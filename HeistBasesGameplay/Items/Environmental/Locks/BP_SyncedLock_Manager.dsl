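; Heist lock manager that syncs a group of environmental locks, so unlocking any one marks the whole group as open.
; Once playlist data is ready it binds to each SyncedLocks entry's unlock dispatcher, unbinding on EndPlay.
; The first unlock unbinds all, and the authority flags locks not yet in the open state to block interaction.
; It then sets ManagerLockState, whose OnRep turns every lock green where cosmetics run.
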(event EventBeginPlay
  (bind _syncedlocks (Variables|Default|GetSyncedLocks))
  ;[0 1 0] BeginPlay - Bind to the unlocked dispatcher for all syncedlocks
  (Class|FortAsyncActionCurrentPlaylistReady|PlaylistDataReadyAsync
    (:OnReady
      (if (Utilities|Array|IsNotEmpty _syncedlocks)
        (for _array_element _syncedlocks
          (Utilities|IsValid _array_element
            (:Is%20Valid
              (Default|BindEventtoOnUnlockedDispatcher _array_element (EventDispatchers|CreateEvent :!event Custom|OnUnlockedAnyLock)))))))))

(event EventEndPlay
  (params (EndPlayReason :type EEndPlayReason%20Enum :default "Destroyed"))
  (bind _syncedlocks (Variables|Default|GetSyncedLocks))
  ;[1 0.1 0.1] EndPlay - Unbind any syncedlocks in array - this array is cleared and events unbound when the lock is opened
  (if (Utilities|Array|IsNotEmpty _syncedlocks)
    (for _array_element _syncedlocks
      (Utilities|IsValid _array_element
        (:Is%20Valid
          (Default|UnbindEventfromOnUnlockedDispatcher _array_element (EventDispatchers|CreateEvent :!event Custom|OnUnlockedAnyLock)))))))

(event Custom|OnUnlockedAnyLock
  (bind _syncedlocks (Variables|Default|GetSyncedLocks))
  ;[0 0 0] SyncLocks - Rough pass on syncing locks. Sets to not allow interact and shows lock open affect
  (if (Utilities|Array|IsNotEmpty _syncedlocks)
    (for _array_element _syncedlocks
      (Utilities|IsValid _array_element
        (:Is%20Valid
          (Default|UnbindEventfromOnUnlockedDispatcher _array_element (EventDispatchers|CreateEvent :!event Custom|OnUnlockedAnyLock))
          (Networking|SwitchHasAuthority
            (:Authority
              (if (Utilities|Enum|NotEqual(Enum) (Variables|/Game/Athena/Items/EnvironmentalItems/Locks/BGA_Athena_Lock_Parent.BGA_Athena_Lock_Parent_C|GetLockState _array_element) "NewEnumerator2")
                (Variables|/Game/Athena/Items/EnvironmentalItems/Locks/BGA_Athena_Lock_Parent.BGA_Athena_Lock_Parent_C|SetEverInteractSetLock true _array_element)))))))
    (Variables|Default|SetManagerLockState "NewEnumerator2") ; Used to set all locks to green light
  ))

(fn OnRep_ManagerLockState
  (bind _syncedlocks (Variables|Default|GetSyncedLocks))
  (Networking|CanExecuteCosmeticEvents
    (:True
      (if (Utilities|Array|IsNotEmpty _syncedlocks)
        (for _array_element _syncedlocks
          (Class|BGAKeysandLocksLock|UpdateColors _array_element (Utilities|Select :Index (Variables|Default|GetManagerLockState) :NewEnumerator0 "NewEnumerator0" :NewEnumerator1 "NewEnumerator0" :NewEnumerator2 "NewEnumerator2")))))))

(fn ConstructionScript)
